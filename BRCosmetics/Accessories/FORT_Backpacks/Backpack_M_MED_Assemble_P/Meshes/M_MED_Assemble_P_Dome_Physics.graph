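# Physics asset for the dome variant of the M_MED_Assemble_P backpack cosmetic, giving its bones simple collision and joint bodies.
# A box covers the main pack bone, and a capsule covers each of the left and right dome bones.
# Two matching constraints join each dome to the main pack, with a wider first swing angle, a narrower second and zero twist angle.
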PhysicsAsset /BRCosmetics/Accessories/FORT_Backpacks/Backpack_M_MED_Assemble_P/Meshes/M_MED_Assemble_P_Dome_Physics
bodies {
 dyn_l_dome [{shapeType=Capsule, length=27.5944, radius=5.2518, center=<7.7552 -0.046 0.3633>, rotation=<rot 0 0 -4.1111>}]
 dyn_mainPack [{shapeType=Box, extentX=11.8677, extentY=12.4942, extentZ=12.1778, center=<0 -3.8912 -0.4964>}]
 dyn_r_dome [{shapeType=Capsule, length=27.9252, radius=5.2518, center=<0.0001 -6.8107 -0.1233>, rotation=<rot 0 0 -4.1111>}]
}
constraints [
 {bone1Name=dyn_l_dome, bone2Name=dyn_mainPack, swing1LimitDegrees=80, swing2LimitDegrees=20, twistLimitDegrees=0}
 {bone1Name=dyn_r_dome, bone2Name=dyn_mainPack, swing1LimitDegrees=80, swing2LimitDegrees=20, twistLimitDegrees=0}
]
