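# Physics asset for the M_MED_Assemble_P backpack cosmetic, defining collision bodies and joints for the pack and its hose.
# A box body covers the main pack bone, and two identical capsules cover the hose base and the following hose bone.
# Angular constraints chain pack, hose base and hose segment, allowing a wide swing on one axis while keeping the other swing and twist tight.
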PhysicsAsset /BRCosmetics/Accessories/FORT_Backpacks/Backpack_M_MED_Assemble_P/Meshes/M_MED_Assemble_P_Pack_Physics
bodies {
 dyn_hose01 [{shapeType=Capsule, length=15.7036, radius=2.6089, shapeName=dyn_hose01_capsule, center=<3.702 0 0>}]
 dyn_hoseBase [{shapeType=Capsule, length=15.7036, radius=2.6089, shapeName=dyn_hoseBase_capsule, center=<3.702 0 0>}]
 dyn_mainPack [{shapeType=Box, extentX=11.8677, extentY=12.4942, extentZ=12.1778, center=<0 -3.8912 -0.4964>}]
}
constraints [
 {bone1Name=dyn_hoseBase, bone2Name=dyn_mainPack, swing1LimitDegrees=50, swing2LimitDegrees=5, twistLimitDegrees=5}
 {bone1Name=dyn_hose01, bone2Name=dyn_hoseBase, swing1LimitDegrees=50, swing2LimitDegrees=5, twistLimitDegrees=5}
]
