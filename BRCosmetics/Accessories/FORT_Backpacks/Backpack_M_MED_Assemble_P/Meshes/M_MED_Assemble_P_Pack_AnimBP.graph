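# Animation Blueprint for the M_MED_Assemble_P backpack cosmetic, adding procedural secondary motion on top of the reference pose.
# Chained AnimDynamics nodes spring the two antenna bones and a needle, with an LOD threshold on the antenna nodes.
# Two partial-alpha RigidBody nodes then simulate separate pack and dome physics assets in root-bone space.
# Final additive bone transforms slightly pitch the left dome and leave the right unchanged; the event graph is unused.
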AnimBlueprint /BRCosmetics/Accessories/FORT_Backpacks/Backpack_M_MED_Assemble_P/Meshes/M_MED_Assemble_P_Pack_AnimBP
target_skeleton=&/M_MED_Assemble_P_Pack_Skeleton
graphs [
 AnimGraph {nodes [
   n1 = Anim Dynamics - Bone: dyn_antennaBase(ComponentPose=Local To Component(LocalPose=Local Space Ref Pose()), Alpha=0.7, ExternalForce="0,0,0") {
    angularDampingOverride=0.9999
    angularSpringConstant=5000
    bAngularSpring=true
    bGravityOverrideInSimSpace=true
    bOverrideAngularDamping=true
    bOverrideLinearDamping=true
    gravityScale=1.1
    lODThreshold=1
    linearDampingOverride=0.9995
    numSolverIterationsPostUpdate=4
    numSolverIterationsPreUpdate=16
    alphaBoolBlend {blendOption=Linear}
    boundBone {boneName=dyn_antennaBase}
    physicsBodyDefinitions [{
      sphereCollisionRadius=0
      boundBone {boneName=dyn_antennaBase}
      boxExtents=<5 15 15>
      constraintSetup {twistAxis=AxisZ, angularLimitsMax=<40 30 5>, angularLimitsMin=<-40 -30 -5>, angularTarget=<0 0 0>}
      localJointOffset=<-2 0 0>
     }]
    relativeSpaceBone {boneName=root}
   }
   n2 = Anim Dynamics - Bone: dyn_antenna01(ComponentPose=@n1, Alpha=0.2, ExternalForce="0,0,0") {
    angularDampingOverride=0.9999
    angularSpringConstant=5000
    bAngularSpring=true
    bGravityOverrideInSimSpace=true
    bOverrideAngularDamping=true
    bOverrideLinearDamping=true
    gravityScale=1.1
    lODThreshold=1
    linearDampingOverride=0.99
    numSolverIterationsPostUpdate=4
    numSolverIterationsPreUpdate=16
    alphaBoolBlend {blendOption=Linear}
    boundBone {boneName=dyn_antenna01}
    physicsBodyDefinitions [{
      sphereCollisionRadius=0
      boundBone {boneName=dyn_antenna01}
      boxExtents=<5 15 15>
      constraintSetup {twistAxis=AxisZ, angularLimitsMax=<40 30 5>, angularLimitsMin=<-40 -30 -5>, angularTarget=<0 0 0>}
      localJointOffset=<-2 0 0>
     }]
    relativeSpaceBone {boneName=root}
   }
   n3 = Anim Dynamics - Bone: dyn_needle(ComponentPose=@n2, Alpha=1, ExternalForce="0,0,0") {
    angularDampingOverride=0.99
    angularSpringConstant=75
    bAngularSpring=true
    bGravityOverrideInSimSpace=true
    bOverrideAngularDamping=true
    gravityScale=0.3
    alphaBoolBlend {blendOption=Linear}
    boundBone {boneName=dyn_needle}
    physicsBodyDefinitions [{
      sphereCollisionRadius=0
      boundBone {boneName=dyn_needle}
      boxExtents=<5 5 5>
      constraintSetup {angularLimitsMax=<2 88.2804 0>, angularLimitsMin=<0 -89.1084 0>, angularTarget=<0 0 0>}
      localJointOffset=<0 0 -1>
     }]
   }
   n4 = RigidBody(ComponentPose=@n3, Alpha=0.4) {
    bClampLinearTranslationLimitToRefPose=true
    bOverrideWorldGravity=true
    overridePhysicsAsset=&/M_MED_Assemble_P_Pack_Physics
    simulationSpace=BaseBoneSpace
    alphaBoolBlend {blendOption=Linear}
    baseBoneRef {boneName=root}
    componentAppliedLinearAccClamp=<5000 5000 2500>
    componentLinearAccScale=<0.45 0.45 0.45>
    overrideWorldGravity {z=-980}
   }
   n5 = RigidBody(ComponentPose=@n4, Alpha=0.2) {
    bClampLinearTranslationLimitToRefPose=true
    bOverrideWorldGravity=true
    overridePhysicsAsset=&/M_MED_Assemble_P_Dome_Physics
    simulationSpace=BaseBoneSpace
    alphaBoolBlend {blendOption=Linear}
    baseBoneRef {boneName=root}
    componentAppliedLinearAccClamp=<5000 5000 2500>
    componentLinearAccScale=<0.45 0.45 0.45>
    overrideWorldGravity {z=-980}
   }
   n6 = "Transform (Modify) Bone - Bone: dyn_l_dome"(ComponentPose=@n5, Alpha=1, Translation="0,0,0", Rotation="-3,0,0", Scale="1,1,1") {rotationMode=BMM_Additive, alphaBoolBlend {blendOption=Linear}, boneToModify {boneName=dyn_l_dome}, rotation {pitch=-3}}
   Output Pose(Result=Component To Local(ComponentPose="Transform (Modify) Bone - Bone: dyn_r_dome"(ComponentPose=@n6, Alpha=1, Translation="0,0,0", Rotation="0,0,0", Scale="1,1,1") {rotationMode=BMM_Additive, alphaBoolBlend {blendOption=Linear}, boneToModify {boneName=dyn_r_dome}})) {layerGroup=None}
  ]}
 EventGraph {nodes [
   TryGetPawnOwner()
   Event BlueprintUpdateAnimation()
  ]}
]
variables []
